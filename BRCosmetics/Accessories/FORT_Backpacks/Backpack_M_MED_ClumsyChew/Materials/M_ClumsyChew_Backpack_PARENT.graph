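# Parent material for the ClumsyChew backpack cosmetic, masked and built on the 2023 character parent PreFX and PostFX functions.
# Between those functions it overrides Opacity Mask to cut away areas flagged by the Tech Art Mask red channel where local-space Z falls below about -10.
# Unflagged areas stay opaque, and the narrow one-unit transition is softened with Fortnite temporal AA dithering.
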Material /BRCosmetics/Accessories/FORT_Backpacks/Backpack_M_MED_ClumsyChew/Materials/M_ClumsyChew_Backpack_PARENT
graph {
 properties {
  bUseMaterialAttributes=true
  bUsedWithMeshDeformer=true
  bUsedWithMorphTargets=true
  bUsedWithSkeletalMesh=true
  blendMode=BLEND_Masked
  pixelDepthOffsetMode=PDOM_Legacy
  shadingModel=MSM_FromMaterialExpression
  usedShadingModels=Default Lit | Subsurface | Clear Coat | Subsurface Profile | Cloth
 }
 nodes [
  n1 = MaterialFunctionCall() {materialFunction=&MF_CharacterParent_2023_PreFX}
  n2 = Saturate(None=Multiply(A=ComponentMask(None=@n1.Tech Art Mask) {r=true}) {constB=10})
  n3 = Add(A=ComponentMask(None=LocalPosition().XYZ) {b=true}) {constB=10}
  n4 = Add(A=OneMinus(None=@n2), B=Multiply(A=@n2, B=Saturate(None=@n3)))
  MaterialFunctionCall(
   Material=SetMaterialAttributes!(
    MaterialAttributes=@n1.Result
    Opacity Mask=MaterialFunctionCall(Alpha Threshold=@n4) {materialFunction=&DitherTemporalAA_Fortnite}.Result
    Emissive Color
   ) {attributeSetTypes [#1, #2]}
   M=@n1.M
   Pre Skinned Local Position=@n1.M
   Distance Blend=@n1.M
   Fresnel=@n1.M
  ) {materialFunction=&MF_CharacterParent_2023_PostFX}
 ]
}
