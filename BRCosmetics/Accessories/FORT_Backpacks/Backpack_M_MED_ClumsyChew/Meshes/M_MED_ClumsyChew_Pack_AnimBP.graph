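# Anim Blueprint for the ClumsyChew back bling: a seven-stem plant that appears to grow over the match.
# AnimDynamics springs jiggle the leaf bases, then chained additive Modify Bone and Constraint nodes retract stems and shrink leaves by per-bone alphas.
# The update event fades those alphas from 1 to 0 using clamped map ranges of match time since Dyn_Time_Start, staggered per bone.
# Front end uses a much shorter growth duration.
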AnimBlueprint /BRCosmetics/Accessories/FORT_Backpacks/Backpack_M_MED_ClumsyChew/Meshes/M_MED_ClumsyChew_Pack_AnimBP
target_skeleton=&/M_MED_ClumsyChew_Pack_Skeleton
graphs [
 AnimGraph {nodes [
   n1 = Anim Dynamics - Bone: dyn_leaf_1_base(ComponentPose=Local To Component(LocalPose=Input Pose()), Alpha=0.9, ExternalForce="0,0,0") {
    angularDampingOverride=0.99
    angularSpringConstant=150
    bAngularSpring=true
    bGravityOverrideInSimSpace=true
    bOverrideAngularDamping=true
    bUseGravityOverride=true
    simulationSpace=BoneRelative
    alphaBoolBlend {blendOption=Linear}
    boundBone {boneName=dyn_leaf_1_base}
    componentAppliedLinearAccClamp=<500 500 500>
    componentLinearAccScale=<0.9 0.9 0.9>
    componentLinearVelScale=<0.9 0.9 0.9>
    gravityOverride {z=-980}
    physicsBodyDefinitions [{
      boundBone {boneName=dyn_leaf_1_base}
      boxExtents=<20 15 10>
      constraintSetup {angularLimitsMax=<10 15 20>, angularLimitsMin=<-10 -30 -45>, angularTarget=<0 0 0>}
      localJointOffset=<-2.5 0 0.1>
     }]
    relativeSpaceBone {boneName=dyn_stem_base}
   }
   n2 = Anim Dynamics - Bone: dyn_leaf_2_base(ComponentPose=@n1, Alpha=1, ExternalForce="0,0,0") {
    angularDampingOverride=0.99
    angularSpringConstant=175
    bAngularSpring=true
    bGravityOverrideInSimSpace=true
    bOverrideAngularDamping=true
    bUseGravityOverride=true
    simulationSpace=BoneRelative
    alphaBoolBlend {blendOption=Linear}
    boundBone {boneName=dyn_leaf_2_base}
    componentAppliedLinearAccClamp=<600 600 600>
    componentLinearAccScale=<0.9 0.9 0.9>
    componentLinearVelScale=<0.9 0.9 0.9>
    gravityOverride {z=-980}
    physicsBodyDefinitions [{
      boundBone {boneName=dyn_leaf_2_base}
      boxExtents=<30 15 20>
      constraintSetup {angularLimitsMax=<20 30 45>, angularLimitsMin=<-20 -30 -45>, angularTarget=<0 0 0>}
      localJointOffset=<-2.25 -0.05 -0.05>
     }]
    relativeSpaceBone {boneName=dyn_stem_base}
   }
   n3 = Anim Dynamics - Bone: dyn_leaf_2_d_base(ComponentPose=@n2, Alpha=1, ExternalForce="0,0,0") {
    angularDampingOverride=0.95
    angularSpringConstant=500
    bAngularSpring=true
    bGravityOverrideInSimSpace=true
    bOverrideAngularDamping=true
    bUseGravityOverride=true
    simulationSpace=BoneRelative
    alphaBoolBlend {blendOption=Linear}
    boundBone {boneName=dyn_leaf_2_d_base}
    componentAppliedLinearAccClamp=<600 600 600>
    componentLinearAccScale=<0.9 0.9 0.9>
    componentLinearVelScale=<0.9 0.9 0.9>
    gravityOverride {z=980}
    physicsBodyDefinitions [{
      boundBone {boneName=dyn_leaf_2_d_base}
      boxExtents=<22.5 12.5 12.5>
      constraintSetup {angularLimitsMax=<20 20 30>, angularLimitsMin=<-20 -45 -30>, angularTarget=<0 0 0>}
      localJointOffset=<0.25 0 1>
     }]
    relativeSpaceBone {boneName=dyn_stem_base}
    simSpaceSettings {maxAngularAcceleration=10, maxAngularVelocity=100, simSpaceAngularAlpha=0.25}
   }
   n4 = Anim Dynamics - Bone: dyn_leaf_3_base(ComponentPose=@n3, Alpha=0.9, ExternalForce="0,0,0") {
    angularDampingOverride=0.999
    angularSpringConstant=200
    bAngularSpring=true
    bGravityOverrideInSimSpace=true
    bOverrideAngularDamping=true
    bUseGravityOverride=true
    simulationSpace=BoneRelative
    alphaBoolBlend {blendOption=Linear}
    boundBone {boneName=dyn_leaf_3_base}
    componentAppliedLinearAccClamp=<500 500 500>
    componentLinearAccScale=<0.9 0.9 0.9>
    componentLinearVelScale=<0.9 0.9 0.9>
    gravityOverride {z=-980}
    physicsBodyDefinitions [{
      boundBone {boneName=dyn_leaf_3_base}
      boxExtents=<25 20 15>
      constraintSetup {angularLimitsMax=<10 30 45>, angularLimitsMin=<-10 -30 -45>, angularTarget=<0 0 0>}
      localJointOffset=<-2 0 0>
     }]
    relativeSpaceBone {boneName=dyn_stem_base}
   }
   n5 = Anim Dynamics - Bone: dyn_leaf_4_base(ComponentPose=@n4, Alpha=1, ExternalForce="0,0,0") {
    angularDampingOverride=0.99
    angularSpringConstant=200
    bAngularSpring=true
    bGravityOverrideInSimSpace=true
    bOverrideAngularDamping=true
    bUseGravityOverride=true
    simulationSpace=BoneRelative
    alphaBoolBlend {blendOption=Linear}
    boundBone {boneName=dyn_leaf_4_base}
    componentAppliedLinearAccClamp=<600 600 600>
    componentLinearAccScale=<0.9 0.9 0.9>
    componentLinearVelScale=<0.9 0.9 0.9>
    gravityOverride {z=-980}
    physicsBodyDefinitions [{
      boundBone {boneName=dyn_leaf_4_base}
      boxExtents=<30 10 20>
      constraintSetup {angularLimitsMax=<20 30 45>, angularLimitsMin=<-20 -30 -45>, angularTarget=<0 0 0>}
      localJointOffset=<-2 -0.05 -0.05>
     }]
    relativeSpaceBone {boneName=dyn_stem_base}
   }
   n6 = Anim Dynamics - Bone: dyn_leaf_4_d_base(ComponentPose=@n5, Alpha=1, ExternalForce="0,0,0") {
    angularDampingOverride=0.99
    angularSpringConstant=200
    bAngularSpring=true
    bGravityOverrideInSimSpace=true
    bOverrideAngularDamping=true
    bUseGravityOverride=true
    simulationSpace=BoneRelative
    alphaBoolBlend {blendOption=Linear}
    boundBone {boneName=dyn_leaf_4_d_base}
    componentAppliedLinearAccClamp=<600 600 600>
    componentLinearAccScale=<0.9 0.9 0.9>
    componentLinearVelScale=<0.9 0.9 0.9>
    gravityOverride {z=-980}
    physicsBodyDefinitions [{
      boundBone {boneName=dyn_leaf_4_d_base}
      boxExtents=<30 10 20>
      constraintSetup {angularLimitsMax=<20 30 45>, angularLimitsMin=<-20 -30 -45>, angularTarget=<0 0 0>}
      localJointOffset=<-2 -0.05 -0.05>
     }]
    relativeSpaceBone {boneName=dyn_stem_base}
   }
   n7 = Anim Dynamics - Bone: dyn_leaf_5_base(ComponentPose=@n6, Alpha=0.9, ExternalForce="0,0,0") {
    angularDampingOverride=0.999
    angularSpringConstant=200
    bAngularSpring=true
    bGravityOverrideInSimSpace=true
    bOverrideAngularDamping=true
    bUseGravityOverride=true
    simulationSpace=BoneRelative
    alphaBoolBlend {blendOption=Linear}
    boundBone {boneName=dyn_leaf_5_base}
    componentAppliedLinearAccClamp=<500 500 500>
    componentLinearAccScale=<0.9 0.9 0.9>
    componentLinearVelScale=<0.9 0.9 0.9>
    gravityOverride {z=-980}
    physicsBodyDefinitions [{
      boundBone {boneName=dyn_leaf_5_base}
      boxExtents=<35 25 20>
      constraintSetup {angularLimitsMax=<10 10 5>, angularLimitsMin=<-10 -30 -45>, angularTarget=<0 0 0>}
      localJointOffset=<-2 -0.25 -0.5>
     }]
    relativeSpaceBone {boneName=dyn_stem_base}
   }
   n8 = Anim Dynamics - Bone: dyn_leaf_6_base(ComponentPose=@n7, Alpha=0.9, ExternalForce="0,0,0") {
    angularDampingOverride=1
    angularSpringConstant=200
    bAngularSpring=true
    bGravityOverrideInSimSpace=true
    bOverrideAngularDamping=true
    bUseGravityOverride=true
    simulationSpace=BoneRelative
    alphaBoolBlend {blendOption=Linear}
    boundBone {boneName=dyn_leaf_6_base}
    componentAppliedLinearAccClamp=<500 500 500>
    componentLinearAccScale=<0.9 0.9 0.9>
    componentLinearVelScale=<0.9 0.9 0.9>
    gravityOverride {z=-980}
    physicsBodyDefinitions [{
      boundBone {boneName=dyn_leaf_6_base}
      boxExtents=<30 20 15>
      constraintSetup {angularLimitsMax=<10 30 45>, angularLimitsMin=<-10 -30 -45>, angularTarget=<0 0 0>}
      localJointOffset=<-2 0 0>
     }]
    relativeSpaceBone {boneName=dyn_stem_base}
   }
   n9 = Anim Dynamics - Bone: dyn_leaf_7_base(ComponentPose=@n8, Alpha=0.9, ExternalForce="0,0,0") {
    angularDampingOverride=1
    angularSpringConstant=200
    bAngularSpring=true
    bGravityOverrideInSimSpace=true
    bOverrideAngularDamping=true
    bUseGravityOverride=true
    simulationSpace=BoneRelative
    alphaBoolBlend {blendOption=Linear}
    boundBone {boneName=dyn_leaf_7_base}
    componentAppliedLinearAccClamp=<500 500 500>
    componentLinearAccScale=<0.9 0.9 0.9>
    componentLinearVelScale=<0.9 0.9 0.9>
    gravityOverride {z=-980}
    physicsBodyDefinitions [{
      boundBone {boneName=dyn_leaf_7_base}
      boxExtents=<27 15 15>
      constraintSetup {angularLimitsMax=<10 30 45>, angularLimitsMin=<-10 -30 -45>, angularTarget=<0 0 0>}
      localJointOffset=<-2 0 0>
     }]
    relativeSpaceBone {boneName=dyn_stem_base}
   }
   n10 = "Transform (Modify) Bone - Bone: dyn_stem_1_1"(ComponentPose=@n9, Alpha=Get Dyn_Stem_1_1() -> (Dyn_Stem_1_1:double), Translation="-5.5,0,0", Rotation="-5,0,0", Scale="1,1,1") {
    rotationMode=BMM_Additive
    rotationSpace=BCS_BoneSpace
    scaleMode=BMM_Additive
    scaleSpace=BCS_BoneSpace
    translationMode=BMM_Additive
    translationSpace=BCS_BoneSpace
    alphaBoolBlend {blendOption=Linear}
    boneToModify {boneName=dyn_stem_1_1}
    rotation {pitch=-5}
    translation {x=-5.5}
   }
   n11 = "Transform (Modify) Bone - Bone: dyn_stem_1_2"(ComponentPose=@n10, Alpha=Get Dyn_Stem_1_2() -> (Dyn_Stem_1_2:double), Translation="-10.5,0,0.5", Rotation="-2.5,-2.5,0", Scale="1,1,1") {
    rotationMode=BMM_Additive
    rotationSpace=BCS_BoneSpace
    scaleMode=BMM_Additive
    scaleSpace=BCS_BoneSpace
    translationMode=BMM_Additive
    translationSpace=BCS_BoneSpace
    alphaBoolBlend {blendOption=Linear}
    boneToModify {boneName=dyn_stem_1_2}
    rotation {pitch=-2.5, yaw=-2.5}
    translation {x=-10.5, z=0.5}
   }
   n12 = "Transform (Modify) Bone - Bone: dyn_stem_1_3"(ComponentPose=@n11, Alpha=Get Dyn_Stem_1_3() -> (Dyn_Stem_1_3:double), Translation="-12.5,0,2.5", Rotation="-10,0,0", Scale="0.5,1,1") {
    rotationMode=BMM_Additive
    rotationSpace=BCS_BoneSpace
    scaleMode=BMM_Additive
    scaleSpace=BCS_BoneSpace
    translationMode=BMM_Additive
    translationSpace=BCS_BoneSpace
    alphaBoolBlend {blendOption=Linear}
    boneToModify {boneName=dyn_stem_1_3}
    rotation {pitch=-10}
    scale {x=0.5}
    translation {x=-12.5, z=2.5}
   }
   n13 = "Constraint - dyn_stem_1_4 (P)"(ComponentPose=@n12, Alpha=1, ConstraintWeights_0=1) {
    alphaBoolBlend {blendOption=Linear}
    boneToModify {boneName=dyn_stem_1_4}
    constraintSetup [{targetBone {boneName=dyn_stem_1_3}}]
    constraintWeights [1]
   }
   n14 = "Transform (Modify) Bone - Bone: dyn_stem_1_4"(ComponentPose=@n13, Alpha=Get Dyn_Stem_1_4() -> (Dyn_Stem_1_4:double), Translation="-8.5,0,0", Rotation="0,0,0", Scale="2,1,1") {
    rotationMode=BMM_Additive
    rotationSpace=BCS_BoneSpace
    scaleMode=BMM_Additive
    scaleSpace=BCS_BoneSpace
    translationMode=BMM_Additive
    translationSpace=BCS_BoneSpace
    alphaBoolBlend {blendOption=Linear}
    boneToModify {boneName=dyn_stem_1_4}
    scale {x=2}
    translation {x=-8.5}
   }
   n15 = "Transform (Modify) Bone - Bone: dyn_stem_2_1"(ComponentPose=@n14, Alpha=Get Dyn_Stem_2_1() -> (Dyn_Stem_2_1:double), Translation="-4.25,0,-1.5", Rotation="10,0,0", Scale="1,1,1") {
    rotationMode=BMM_Additive
    rotationSpace=BCS_BoneSpace
    scaleMode=BMM_Additive
    scaleSpace=BCS_BoneSpace
    translationMode=BMM_Additive
    translationSpace=BCS_BoneSpace
    alphaBoolBlend {blendOption=Linear}
    boneToModify {boneName=dyn_stem_2_1}
    rotation {pitch=10}
    translation {x=-4.25, z=-1.5}
   }
   n16 = "Transform (Modify) Bone - Bone: dyn_stem_2_2"(ComponentPose=@n15, Alpha=Get Dyn_Stem_2_2() -> (Dyn_Stem_2_2:double), Translation="-8.5,0,-1.4", Rotation="12,0,0", Scale="1,1,1") {
    rotationMode=BMM_Additive
    rotationSpace=BCS_BoneSpace
    scaleMode=BMM_Additive
    scaleSpace=BCS_BoneSpace
    translationMode=BMM_Additive
    translationSpace=BCS_BoneSpace
    alphaBoolBlend {blendOption=Linear}
    boneToModify {boneName=dyn_stem_2_2}
    rotation {pitch=12}
    translation {x=-8.5, z=-1.4}
   }
   n17 = "Transform (Modify) Bone - Bone: dyn_stem_2_3"(ComponentPose=@n16, Alpha=Get Dyn_Stem_2_3() -> (Dyn_Stem_2_3:double), Translation="-8.5,0,-2", Rotation="20,0,0", Scale="1,1,1") {
    rotationMode=BMM_Additive
    rotationSpace=BCS_BoneSpace
    scaleMode=BMM_Additive
    scaleSpace=BCS_BoneSpace
    translationMode=BMM_Additive
    translationSpace=BCS_BoneSpace
    alphaBoolBlend {blendOption=Linear}
    boneToModify {boneName=dyn_stem_2_3}
    rotation {pitch=20}
    translation {x=-8.5, z=-2}
   }
   n18 = "Constraint - dyn_stem_2_4 (P)"(ComponentPose=@n17, Alpha=1, ConstraintWeights_0=1) {
    alphaBoolBlend {blendOption=Linear}
    boneToModify {boneName=dyn_stem_2_4}
    constraintSetup [{targetBone {boneName=dyn_stem_2_3}}]
    constraintWeights [1]
   }
   n19 = "Transform (Modify) Bone - Bone: dyn_stem_2_4"(ComponentPose=@n18, Alpha=Get Dyn_Stem_2_4() -> (Dyn_Stem_2_4:double), Translation="0,0,0", Rotation="0,0,0", Scale="1,1,1") {
    rotationMode=BMM_Additive
    rotationSpace=BCS_BoneSpace
    scaleMode=BMM_Additive
    scaleSpace=BCS_BoneSpace
    translationMode=BMM_Additive
    translationSpace=BCS_BoneSpace
    alphaBoolBlend {blendOption=Linear}
    boneToModify {boneName=dyn_stem_2_4}
   }
   n20 = "Transform (Modify) Bone - Bone: dyn_stem_3_1"(ComponentPose=@n19, Alpha=Get Dyn_Stem_3_1() -> (Dyn_Stem_3_1:double), Translation="-4.5,0,0", Rotation="0,0,0", Scale="1,1,1") {
    rotationMode=BMM_Additive
    rotationSpace=BCS_BoneSpace
    scaleMode=BMM_Additive
    scaleSpace=BCS_BoneSpace
    translationMode=BMM_Additive
    translationSpace=BCS_BoneSpace
    alphaBoolBlend {blendOption=Linear}
    boneToModify {boneName=dyn_stem_3_1}
    translation {x=-4.5}
   }
   n21 = "Transform (Modify) Bone - Bone: dyn_stem_3_2"(ComponentPose=@n20, Alpha=Get Dyn_Stem_3_2() -> (Dyn_Stem_3_2:double), Translation="-8.75,0,0", Rotation="0,0,0", Scale="1,1,1") {
    rotationMode=BMM_Additive
    rotationSpace=BCS_BoneSpace
    scaleMode=BMM_Additive
    scaleSpace=BCS_BoneSpace
    translationMode=BMM_Additive
    translationSpace=BCS_BoneSpace
    alphaBoolBlend {blendOption=Linear}
    boneToModify {boneName=dyn_stem_3_2}
    translation {x=-8.75}
   }
   n22 = "Transform (Modify) Bone - Bone: dyn_stem_3_3"(ComponentPose=@n21, Alpha=Get Dyn_Stem_3_3() -> (Dyn_Stem_3_3:double), Translation="-12.75,-0.25,-1.25", Rotation="5,0,0", Scale="1,1,1") {
    rotationMode=BMM_Additive
    rotationSpace=BCS_BoneSpace
    scaleMode=BMM_Additive
    scaleSpace=BCS_BoneSpace
    translationMode=BMM_Additive
    translationSpace=BCS_BoneSpace
    alphaBoolBlend {blendOption=Linear}
    boneToModify {boneName=dyn_stem_3_3}
    rotation {pitch=5}
    translation=<-12.75 -0.25 -1.25>
   }
   n23 = "Constraint - dyn_stem_3_4 (P)"(ComponentPose=@n22, Alpha=1, ConstraintWeights_0=1) {
    alphaBoolBlend {blendOption=Linear}
    boneToModify {boneName=dyn_stem_3_4}
    constraintSetup [{targetBone {boneName=dyn_stem_3_3}}]
    constraintWeights [1]
   }
   n24 = "Transform (Modify) Bone - Bone: dyn_stem_3_4"(ComponentPose=@n23, Alpha=Get Dyn_Stem_3_4() -> (Dyn_Stem_3_4:double), Translation="-3.5,0,0", Rotation="5,0,0", Scale="1,1,1") {
    rotationMode=BMM_Additive
    rotationSpace=BCS_BoneSpace
    scaleMode=BMM_Additive
    scaleSpace=BCS_BoneSpace
    translationMode=BMM_Additive
    translationSpace=BCS_BoneSpace
    alphaBoolBlend {blendOption=Linear}
    boneToModify {boneName=dyn_stem_3_4}
    rotation {pitch=5}
    translation {x=-3.5}
   }
   n25 = "Transform (Modify) Bone - Bone: dyn_stem_4_1"(ComponentPose=@n24, Alpha=Get Dyn_Stem_4_1() -> (Dyn_Stem_4_1:double), Translation="-4.5,0,0", Rotation="-10,5,0", Scale="1,1,1") {
    rotationMode=BMM_Additive
    rotationSpace=BCS_BoneSpace
    scaleMode=BMM_Additive
    scaleSpace=BCS_BoneSpace
    translationMode=BMM_Additive
    translationSpace=BCS_BoneSpace
    alphaBoolBlend {blendOption=Linear}
    boneToModify {boneName=dyn_stem_4_1}
    rotation {pitch=-10, yaw=5}
    translation {x=-4.5}
   }
   n26 = "Transform (Modify) Bone - Bone: dyn_stem_4_2"(ComponentPose=@n25, Alpha=Get Dyn_Stem_4_2() -> (Dyn_Stem_4_2:double), Translation="-7.5,-0.25,1", Rotation="-10,0,0", Scale="1,1,1") {
    rotationMode=BMM_Additive
    rotationSpace=BCS_BoneSpace
    scaleMode=BMM_Additive
    scaleSpace=BCS_BoneSpace
    translationMode=BMM_Additive
    translationSpace=BCS_BoneSpace
    alphaBoolBlend {blendOption=Linear}
    boneToModify {boneName=dyn_stem_4_2}
    rotation {pitch=-10}
    translation=<-7.5 -0.25 1>
   }
   n27 = "Transform (Modify) Bone - Bone: dyn_stem_4_3"(ComponentPose=@n26, Alpha=Get Dyn_Stem_4_3() -> (Dyn_Stem_4_3:double), Translation="-7.5,-0.5,1.5", Rotation="-15,5,0", Scale="1,1,1") {
    rotationMode=BMM_Additive
    rotationSpace=BCS_BoneSpace
    scaleMode=BMM_Additive
    scaleSpace=BCS_BoneSpace
    translationMode=BMM_Additive
    translationSpace=BCS_BoneSpace
    alphaBoolBlend {blendOption=Linear}
    boneToModify {boneName=dyn_stem_4_3}
    rotation {pitch=-15, yaw=5}
    translation=<-7.5 -0.5 1.5>
   }
   n28 = "Constraint - dyn_stem_4_4 (P)"(ComponentPose=@n27, Alpha=1, ConstraintWeights_0=1) {
    alphaBoolBlend {blendOption=Linear}
    boneToModify {boneName=dyn_stem_4_4}
    constraintSetup [{targetBone {boneName=dyn_stem_4_3}}]
    constraintWeights [1]
   }
   n29 = "Transform (Modify) Bone - Bone: dyn_stem_4_4"(ComponentPose=@n28, Alpha=Get Dyn_Stem_4_4() -> (Dyn_Stem_4_4:double), Translation="0,0,0", Rotation="0,0,0", Scale="1,1,1") {
    rotationMode=BMM_Additive
    rotationSpace=BCS_BoneSpace
    scaleMode=BMM_Additive
    scaleSpace=BCS_BoneSpace
    translationMode=BMM_Additive
    translationSpace=BCS_BoneSpace
    alphaBoolBlend {blendOption=Linear}
    boneToModify {boneName=dyn_stem_4_4}
   }
   n30 = "Transform (Modify) Bone - Bone: dyn_stem_5_1"(ComponentPose=@n29, Alpha=Get Dyn_Stem_5_1() -> (Dyn_Stem_5_1:double), Translation="-4.5,0,0", Rotation="0,0,0", Scale="1,1,1") {
    rotationMode=BMM_Additive
    rotationSpace=BCS_BoneSpace
    scaleMode=BMM_Additive
    scaleSpace=BCS_BoneSpace
    translationMode=BMM_Additive
    translationSpace=BCS_BoneSpace
    alphaBoolBlend {blendOption=Linear}
    boneToModify {boneName=dyn_stem_5_1}
    translation {x=-4.5}
   }
   n31 = "Transform (Modify) Bone - Bone: dyn_stem_5_2"(ComponentPose=@n30, Alpha=Get Dyn_Stem_5_2() -> (Dyn_Stem_5_2:double), Translation="-9,0,-2.3", Rotation="15,0,0", Scale="1,1,1") {
    rotationMode=BMM_Additive
    rotationSpace=BCS_BoneSpace
    scaleMode=BMM_Additive
    scaleSpace=BCS_BoneSpace
    translationMode=BMM_Additive
    translationSpace=BCS_BoneSpace
    alphaBoolBlend {blendOption=Linear}
    boneToModify {boneName=dyn_stem_5_2}
    rotation {pitch=15}
    translation {x=-9, z=-2.3}
   }
   n32 = "Transform (Modify) Bone - Bone: dyn_stem_5_3"(ComponentPose=@n31, Alpha=Get Dyn_Stem_5_3() -> (Dyn_Stem_5_3:double), Translation="-13.2,-1,-3", Rotation="22.5,5,0", Scale="1,1,1") {
    rotationMode=BMM_Additive
    rotationSpace=BCS_BoneSpace
    scaleMode=BMM_Additive
    scaleSpace=BCS_BoneSpace
    translationMode=BMM_Additive
    translationSpace=BCS_BoneSpace
    alphaBoolBlend {blendOption=Linear}
    boneToModify {boneName=dyn_stem_5_3}
    rotation {pitch=22.5, yaw=5}
    translation=<-13.2 -1 -3>
   }
   n33 = "Constraint - dyn_stem_5_4 (P)"(ComponentPose=@n32, Alpha=1, ConstraintWeights_0=1) {
    alphaBoolBlend {blendOption=Linear}
    boneToModify {boneName=dyn_stem_5_4}
    constraintSetup [{targetBone {boneName=dyn_stem_5_3}}]
    constraintWeights [1]
   }
   n34 = "Transform (Modify) Bone - Bone: dyn_stem_5_4"(ComponentPose=@n33, Alpha=Get Dyn_Stem_5_4() -> (Dyn_Stem_5_4:double), Translation="-4,0,0", Rotation="0,0,0", Scale="1,1,1") {
    rotationMode=BMM_Additive
    rotationSpace=BCS_BoneSpace
    scaleMode=BMM_Additive
    scaleSpace=BCS_BoneSpace
    translationMode=BMM_Additive
    translationSpace=BCS_BoneSpace
    alphaBoolBlend {blendOption=Linear}
    boneToModify {boneName=dyn_stem_5_4}
    translation {x=-4}
   }
   n35 = "Transform (Modify) Bone - Bone: dyn_stem_6_1"(ComponentPose=@n34, Alpha=Get Dyn_Stem_6_1() -> (Dyn_Stem_6_1:double), Translation="-5.5,0,0", Rotation="0,0,0", Scale="1,1,1") {
    rotationMode=BMM_Additive
    rotationSpace=BCS_BoneSpace
    scaleMode=BMM_Additive
    scaleSpace=BCS_BoneSpace
    translationMode=BMM_Additive
    translationSpace=BCS_BoneSpace
    alphaBoolBlend {blendOption=Linear}
    boneToModify {boneName=dyn_stem_6_1}
    translation {x=-5.5}
   }
   n36 = "Transform (Modify) Bone - Bone: dyn_stem_6_2"(ComponentPose=@n35, Alpha=Get Dyn_Stem_6_2() -> (Dyn_Stem_6_2:double), Translation="-10,0.25,-0.5", Rotation="0,0,0", Scale="1,1,1") {
    rotationMode=BMM_Additive
    rotationSpace=BCS_BoneSpace
    scaleMode=BMM_Additive
    scaleSpace=BCS_BoneSpace
    translationMode=BMM_Additive
    translationSpace=BCS_BoneSpace
    alphaBoolBlend {blendOption=Linear}
    boneToModify {boneName=dyn_stem_6_2}
    translation=<-10 0.25 -0.5>
   }
   n37 = "Transform (Modify) Bone - Bone: dyn_stem_6_3"(ComponentPose=@n36, Alpha=Get Dyn_Stem_6_3() -> (Dyn_Stem_6_3:double), Translation="-14.25,0,-1.25", Rotation="-2.5,5,0", Scale="1,1,1") {
    rotationMode=BMM_Additive
    rotationSpace=BCS_BoneSpace
    scaleMode=BMM_Additive
    scaleSpace=BCS_BoneSpace
    translationMode=BMM_Additive
    translationSpace=BCS_BoneSpace
    alphaBoolBlend {blendOption=Linear}
    boneToModify {boneName=dyn_stem_6_3}
    rotation {pitch=-2.5, yaw=5}
    translation {x=-14.25, z=-1.25}
   }
   n38 = "Constraint - dyn_stem_6_4 (P)"(ComponentPose=@n37, Alpha=1, ConstraintWeights_0=1) {
    alphaBoolBlend {blendOption=Linear}
    boneToModify {boneName=dyn_stem_6_4}
    constraintSetup [{targetBone {boneName=dyn_stem_6_3}}]
    constraintWeights [1]
   }
   n39 = "Transform (Modify) Bone - Bone: dyn_stem_6_4"(ComponentPose=@n38, Alpha=Get Dyn_Stem_6_4() -> (Dyn_Stem_6_4:double), Translation="-4.25,0,0", Rotation="0,0,0", Scale="1,1,1") {
    rotationMode=BMM_Additive
    rotationSpace=BCS_BoneSpace
    scaleMode=BMM_Additive
    scaleSpace=BCS_BoneSpace
    translationMode=BMM_Additive
    translationSpace=BCS_BoneSpace
    alphaBoolBlend {blendOption=Linear}
    boneToModify {boneName=dyn_stem_6_4}
    translation {x=-4.25}
   }
   n40 = "Transform (Modify) Bone - Bone: dyn_stem_7_1"(ComponentPose=@n39, Alpha=Get Dyn_Stem_7_1() -> (Dyn_Stem_7_1:double), Translation="-4.75,0,0", Rotation="0,0,0", Scale="1,1,1") {
    rotationMode=BMM_Additive
    rotationSpace=BCS_BoneSpace
    scaleMode=BMM_Additive
    scaleSpace=BCS_BoneSpace
    translationMode=BMM_Additive
    translationSpace=BCS_BoneSpace
    alphaBoolBlend {blendOption=Linear}
    boneToModify {boneName=dyn_stem_7_1}
    rotation=<rot 0 0 0>
    translation {x=-4.75}
   }
   n41 = "Transform (Modify) Bone - Bone: dyn_stem_7_2"(
    ComponentPose=@n40
    Alpha=Get Dyn_Stem_7_2() -> (Dyn_Stem_7_2:double)
    Translation="-9.25,-0.25,1"
    Rotation="-5.000,-0.000,0.000"
    Scale="1,1,1"
   ) {
    rotationMode=BMM_Additive
    rotationSpace=BCS_BoneSpace
    scaleMode=BMM_Additive
    scaleSpace=BCS_BoneSpace
    translationMode=BMM_Additive
    translationSpace=BCS_BoneSpace
    alphaBoolBlend {blendOption=Linear}
    boneToModify {boneName=dyn_stem_7_2}
    rotation=<rot -5 0 0>
    translation=<-9.25 -0.25 1>
   }
   n42 = "Transform (Modify) Bone - Bone: dyn_stem_7_3"(ComponentPose=@n41, Alpha=Get Dyn_Stem_7_3() -> (Dyn_Stem_7_3:double), Translation="-13.25,0,2.5", Rotation="-15,0,0", Scale="1,1,1") {
    rotationMode=BMM_Additive
    rotationSpace=BCS_BoneSpace
    scaleMode=BMM_Additive
    scaleSpace=BCS_BoneSpace
    translationMode=BMM_Additive
    translationSpace=BCS_BoneSpace
    alphaBoolBlend {blendOption=Linear}
    boneToModify {boneName=dyn_stem_7_3}
    rotation {pitch=-15}
    translation {x=-13.25, z=2.5}
   }
   n43 = "Constraint - dyn_stem_7_4 (P)"(ComponentPose=@n42, Alpha=1, ConstraintWeights_0=1) {
    alphaBoolBlend {blendOption=Linear}
    boneToModify {boneName=dyn_stem_7_4}
    constraintSetup [{targetBone {boneName=dyn_stem_7_3}}]
    constraintWeights [1]
   }
   n44 = "Transform (Modify) Bone - Bone: dyn_stem_7_4"(ComponentPose=@n43, Alpha=Get Dyn_Stem_7_4() -> (Dyn_Stem_7_4:double), Translation="-4,0,0", Rotation="0,0,0", Scale="1,1,1") {
    rotationMode=BMM_Additive
    rotationSpace=BCS_BoneSpace
    scaleMode=BMM_Additive
    scaleSpace=BCS_BoneSpace
    translationMode=BMM_Additive
    translationSpace=BCS_BoneSpace
    alphaBoolBlend {blendOption=Linear}
    boneToModify {boneName=dyn_stem_7_4}
    rotation=<rot -0.0001 0 0>
    translation {x=-4}
   }
   n45 = "Transform (Modify) Bone - Bone: dyn_leaf_1_1"(ComponentPose=@n44, Alpha=Get Dyn_Leaf_1_1() -> (Dyn_Leaf_1_1:double), Translation="0,0,0", Rotation="0,-60,0", Scale="0.1,0.1,0.1") {
    rotationMode=BMM_Additive
    rotationSpace=BCS_BoneSpace
    scaleMode=BMM_Additive
    scaleSpace=BCS_BoneSpace
    translationMode=BMM_Additive
    translationSpace=BCS_BoneSpace
    alphaBoolBlend {blendOption=Linear}
    boneToModify {boneName=dyn_leaf_1_1}
    rotation {yaw=-60}
    scale=<0.1 0.1 0.1>
   }
   n46 = "Transform (Modify) Bone - Bone: dyn_leaf_1_2"(ComponentPose=@n45, Alpha=Get Dyn_Leaf_1_2() -> (Dyn_Leaf_1_2:double), Translation="0,0,0", Rotation="0,-30,0", Scale="0.1,0.1,0.1") {
    rotationMode=BMM_Additive
    rotationSpace=BCS_BoneSpace
    scaleMode=BMM_Additive
    scaleSpace=BCS_BoneSpace
    translationMode=BMM_Additive
    translationSpace=BCS_BoneSpace
    alphaBoolBlend {blendOption=Linear}
    boneToModify {boneName=dyn_leaf_1_2}
    rotation {yaw=-30}
    scale=<0.1 0.1 0.1>
   }
   n47 = "Transform (Modify) Bone - Bone: dyn_leaf_2_1"(ComponentPose=@n46, Alpha=Get Dyn_Leaf_2_1() -> (Dyn_Leaf_2_1:double), Translation="0,0,0", Rotation="20,30,0", Scale="0.1,0.1,0.1") {
    rotationMode=BMM_Additive
    rotationSpace=BCS_BoneSpace
    scaleMode=BMM_Additive
    scaleSpace=BCS_BoneSpace
    translationMode=BMM_Additive
    translationSpace=BCS_BoneSpace
    alphaBoolBlend {blendOption=Linear}
    boneToModify {boneName=dyn_leaf_2_1}
    rotation {pitch=20, yaw=30}
    scale=<0.1 0.1 0.1>
   }
   n48 = "Transform (Modify) Bone - Bone: dyn_leaf_2_2"(ComponentPose=@n47, Alpha=Get Dyn_Leaf_2_2() -> (Dyn_Leaf_2_2:double), Translation="0,0,0", Rotation="0,0,0", Scale="0.5,0.5,0.5") {
    rotationMode=BMM_Additive
    rotationSpace=BCS_BoneSpace
    scaleMode=BMM_Additive
    scaleSpace=BCS_BoneSpace
    translationMode=BMM_Additive
    translationSpace=BCS_BoneSpace
    alphaBoolBlend {blendOption=Linear}
    boneToModify {boneName=dyn_leaf_2_2}
    scale=<0.5 0.5 0.5>
   }
   n49 = "Transform (Modify) Bone - Bone: dyn_leaf_2_3"(ComponentPose=@n48, Alpha=Get Dyn_Leaf_2_3() -> (Dyn_Leaf_2_3:double), Translation="0,0,0", Rotation="0,-30,0", Scale="0.1,0.1,0.1") {
    rotationMode=BMM_Additive
    rotationSpace=BCS_BoneSpace
    scaleMode=BMM_Additive
    scaleSpace=BCS_BoneSpace
    translationMode=BMM_Additive
    translationSpace=BCS_BoneSpace
    alphaBoolBlend {blendOption=Linear}
    boneToModify {boneName=dyn_leaf_2_3}
    rotation {yaw=-30}
    scale=<0.1 0.1 0.1>
   }
   n50 = "Transform (Modify) Bone - Bone: dyn_leaf_2_d_base"(ComponentPose=@n49, Alpha=Get Dyn_Leaf_2_base() -> (Dyn_Leaf_2_base:double), Translation="0,0,0", Rotation="25,0,0", Scale="0.1,0.1,0.1") {
    rotationMode=BMM_Additive
    rotationSpace=BCS_BoneSpace
    scaleMode=BMM_Additive
    scaleSpace=BCS_BoneSpace
    translationMode=BMM_Additive
    translationSpace=BCS_BoneSpace
    alphaBoolBlend {blendOption=Linear}
    boneToModify {boneName=dyn_leaf_2_d_base}
    rotation {pitch=25}
    scale=<0.1 0.1 0.1>
   }
   n51 = "Transform (Modify) Bone - Bone: dyn_leaf_2_d_1"(ComponentPose=@n50, Alpha=Get Dyn_Leaf_2_d_1() -> (Dyn_Leaf_2_d_1:double), Translation="0,0,0", Rotation="25,0,0", Scale="0.1,0.1,0.1") {
    rotationMode=BMM_Additive
    rotationSpace=BCS_BoneSpace
    scaleMode=BMM_Additive
    scaleSpace=BCS_BoneSpace
    translationMode=BMM_Additive
    translationSpace=BCS_BoneSpace
    alphaBoolBlend {blendOption=Linear}
    boneToModify {boneName=dyn_leaf_2_d_1}
    rotation {pitch=25}
    scale=<0.1 0.1 0.1>
   }
   n52 = "Transform (Modify) Bone - Bone: dyn_leaf_2_d_2"(ComponentPose=@n51, Alpha=Get Dyn_Leaf_2_d_2() -> (Dyn_Leaf_2_d_2:double), Translation="0,0,0", Rotation="0,0,0", Scale="0.1,0.1,0.1") {
    rotationMode=BMM_Additive
    rotationSpace=BCS_BoneSpace
    scaleMode=BMM_Additive
    scaleSpace=BCS_BoneSpace
    translationMode=BMM_Additive
    translationSpace=BCS_BoneSpace
    alphaBoolBlend {blendOption=Linear}
    boneToModify {boneName=dyn_leaf_2_d_2}
    scale=<0.1 0.1 0.1>
   }
   n53 = "Transform (Modify) Bone - Bone: dyn_leaf_3_1"(ComponentPose=@n52, Alpha=Get Dyn_Leaf_3_1() -> (Dyn_Leaf_3_1:double), Translation="0,0,0", Rotation="0,-60,0", Scale="0.1,0.1,0.1") {
    rotationMode=BMM_Additive
    rotationSpace=BCS_BoneSpace
    scaleMode=BMM_Additive
    scaleSpace=BCS_BoneSpace
    translationMode=BMM_Additive
    translationSpace=BCS_BoneSpace
    alphaBoolBlend {blendOption=Linear}
    boneToModify {boneName=dyn_leaf_3_1}
    rotation {yaw=-60}
    scale=<0.1 0.1 0.1>
   }
   n54 = "Transform (Modify) Bone - Bone: dyn_leaf_4_1"(ComponentPose=@n53, Alpha=Get Dyn_Leaf_4_1() -> (Dyn_Leaf_4_1:double), Translation="0,0,0", Rotation="-20,30,0", Scale="0.1,0.1,0.1") {
    rotationMode=BMM_Additive
    rotationSpace=BCS_BoneSpace
    scaleMode=BMM_Additive
    scaleSpace=BCS_BoneSpace
    translationMode=BMM_Additive
    translationSpace=BCS_BoneSpace
    alphaBoolBlend {blendOption=Linear}
    boneToModify {boneName=dyn_leaf_4_1}
    rotation {pitch=-20, yaw=30}
    scale=<0.1 0.1 0.1>
   }
   n55 = "Transform (Modify) Bone - Bone: dyn_leaf_4_2"(ComponentPose=@n54, Alpha=Get Dyn_Leaf_4_2() -> (Dyn_Leaf_4_2:double), Translation="0,0,0", Rotation="0,30,0", Scale="0.1,0.1,0.1") {
    rotationMode=BMM_Additive
    rotationSpace=BCS_BoneSpace
    scaleMode=BMM_Additive
    scaleSpace=BCS_BoneSpace
    translationMode=BMM_Additive
    translationSpace=BCS_BoneSpace
    alphaBoolBlend {blendOption=Linear}
    boneToModify {boneName=dyn_leaf_4_2}
    rotation {yaw=30}
    scale=<0.1 0.1 0.1>
   }
   n56 = "Transform (Modify) Bone - Bone: dyn_leaf_4_3"(ComponentPose=@n55, Alpha=Get Dyn_Leaf_4_3() -> (Dyn_Leaf_4_3:double), Translation="0,0,0", Rotation="-10,0,0", Scale="0.5,0.5,0.5") {
    rotationMode=BMM_Additive
    rotationSpace=BCS_BoneSpace
    scaleMode=BMM_Additive
    scaleSpace=BCS_BoneSpace
    translationMode=BMM_Additive
    translationSpace=BCS_BoneSpace
    alphaBoolBlend {blendOption=Linear}
    boneToModify {boneName=dyn_leaf_4_3}
    rotation {pitch=-10}
    scale=<0.5 0.5 0.5>
   }
   n57 = "Transform (Modify) Bone - Bone: dyn_leaf_4_d_base"(ComponentPose=@n56, Alpha=Get Dyn_Leaf_4_base() -> (Dyn_Leaf_4_base:double), Translation="0,0,0", Rotation="-15,0,0", Scale="0.5,0.5,0.5") {
    rotationMode=BMM_Additive
    rotationSpace=BCS_BoneSpace
    scaleMode=BMM_Additive
    scaleSpace=BCS_BoneSpace
    translationMode=BMM_Additive
    translationSpace=BCS_BoneSpace
    alphaBoolBlend {blendOption=Linear}
    boneToModify {boneName=dyn_leaf_4_d_base}
    rotation {pitch=-15}
    scale=<0.5 0.5 0.5>
   }
   n58 = "Transform (Modify) Bone - Bone: dyn_leaf_4_1_d_base"(ComponentPose=@n57, Alpha=Get Dyn_Leaf_4_d_1() -> (Dyn_Leaf_4_d_1:double), Translation="0,0,0", Rotation="-30,0,0", Scale="0.1,0.1,0.1") {
    rotationMode=BMM_Additive
    rotationSpace=BCS_BoneSpace
    scaleMode=BMM_Additive
    scaleSpace=BCS_BoneSpace
    translationMode=BMM_Additive
    translationSpace=BCS_BoneSpace
    alphaBoolBlend {blendOption=Linear}
    boneToModify {boneName=dyn_leaf_4_1_d_base}
    rotation {pitch=-30}
    scale=<0.1 0.1 0.1>
   }
   n59 = "Transform (Modify) Bone - Bone: dyn_leaf_5_base"(ComponentPose=@n58, Alpha=Get Dyn_Leaf_5_base() -> (Dyn_Leaf_5_base:double), Translation="0,0,0", Rotation="30,-45,0", Scale="0.2,0.75,0.75") {
    rotationMode=BMM_Additive
    rotationSpace=BCS_BoneSpace
    scaleMode=BMM_Additive
    scaleSpace=BCS_BoneSpace
    translationMode=BMM_Additive
    translationSpace=BCS_BoneSpace
    alphaBoolBlend {blendOption=Linear}
    boneToModify {boneName=dyn_leaf_5_base}
    rotation {pitch=30, yaw=-45}
    scale=<0.2 0.75 0.75>
   }
   n60 = "Transform (Modify) Bone - Bone: dyn_leaf_5_1"(ComponentPose=@n59, Alpha=Get Dyn_Leaf_5_1() -> (Dyn_Leaf_5_1:double), Translation="0,0,0", Rotation="0,0,0", Scale="0.1,0.1,0.1") {
    rotationMode=BMM_Additive
    rotationSpace=BCS_BoneSpace
    scaleMode=BMM_Additive
    scaleSpace=BCS_BoneSpace
    translationMode=BMM_Additive
    translationSpace=BCS_BoneSpace
    alphaBoolBlend {blendOption=Linear}
    boneToModify {boneName=dyn_leaf_5_1}
    scale=<0.1 0.1 0.1>
   }
   n61 = "Transform (Modify) Bone - Bone: dyn_leaf_5_2"(ComponentPose=@n60, Alpha=Get Dyn_Leaf_5_2() -> (Dyn_Leaf_5_2:double), Translation="0,0,0", Rotation="0,-30,0", Scale="0.1,0.1,0.1") {
    rotationMode=BMM_Additive
    rotationSpace=BCS_BoneSpace
    scaleMode=BMM_Additive
    scaleSpace=BCS_BoneSpace
    translationMode=BMM_Additive
    translationSpace=BCS_BoneSpace
    alphaBoolBlend {blendOption=Linear}
    boneToModify {boneName=dyn_leaf_5_2}
    rotation {yaw=-30}
    scale=<0.1 0.1 0.1>
   }
   n62 = "Transform (Modify) Bone - Bone: dyn_leaf_5_3"(ComponentPose=@n61, Alpha=Get Dyn_Leaf_5_3() -> (Dyn_Leaf_5_3:double), Translation="0,0,0", Rotation="0,-30,0", Scale="0.1,0.1,0.1") {
    rotationMode=BMM_Additive
    rotationSpace=BCS_BoneSpace
    scaleMode=BMM_Additive
    scaleSpace=BCS_BoneSpace
    translationMode=BMM_Additive
    translationSpace=BCS_BoneSpace
    alphaBoolBlend {blendOption=Linear}
    boneToModify {boneName=dyn_leaf_5_3}
    rotation {yaw=-30}
    scale=<0.1 0.1 0.1>
   }
   n63 = "Transform (Modify) Bone - Bone: dyn_leaf_6_1"(ComponentPose=@n62, Alpha=Get Dyn_Leaf_6_1() -> (Dyn_Leaf_6_1:double), Translation="0,0,0", Rotation="0,0,0", Scale="0.1,0.1,0.1") {
    rotationMode=BMM_Additive
    rotationSpace=BCS_BoneSpace
    scaleMode=BMM_Additive
    scaleSpace=BCS_BoneSpace
    translationMode=BMM_Additive
    translationSpace=BCS_BoneSpace
    alphaBoolBlend {blendOption=Linear}
    boneToModify {boneName=dyn_leaf_6_1}
    scale=<0.1 0.1 0.1>
   }
   n64 = "Transform (Modify) Bone - Bone: dyn_leaf_6_2"(ComponentPose=@n63, Alpha=Get Dyn_Leaf_6_2() -> (Dyn_Leaf_6_2:double), Translation="0,0,0", Rotation="0,-30,0", Scale="0.1,0.1,0.1") {
    rotationMode=BMM_Additive
    rotationSpace=BCS_BoneSpace
    scaleMode=BMM_Additive
    scaleSpace=BCS_BoneSpace
    translationMode=BMM_Additive
    translationSpace=BCS_BoneSpace
    alphaBoolBlend {blendOption=Linear}
    boneToModify {boneName=dyn_leaf_6_2}
    rotation {yaw=-30}
    scale=<0.1 0.1 0.1>
   }
   n65 = "Transform (Modify) Bone - Bone: dyn_leaf_7_1"(ComponentPose=@n64, Alpha=Get Dyn_Leaf_7_1() -> (Dyn_Leaf_7_1:double), Translation="0,0,0", Rotation="0,-30,0", Scale="0.1,0.1,0.1") {
    rotationMode=BMM_Additive
    rotationSpace=BCS_BoneSpace
    scaleMode=BMM_Additive
    scaleSpace=BCS_BoneSpace
    translationMode=BMM_Additive
    translationSpace=BCS_BoneSpace
    alphaBoolBlend {blendOption=Linear}
    boneToModify {boneName=dyn_leaf_7_1}
    rotation {yaw=-30}
    scale=<0.1 0.1 0.1>
   }
   n66 = "Transform (Modify) Bone - Bone: dyn_leaf_7_2"(ComponentPose=@n65, Alpha=Get Dyn_Leaf_7_2() -> (Dyn_Leaf_7_2:double), Translation="0,0,0", Rotation="0,-15,0", Scale="0.1,0.1,0.1") {
    rotationMode=BMM_Additive
    rotationSpace=BCS_BoneSpace
    scaleMode=BMM_Additive
    scaleSpace=BCS_BoneSpace
    translationMode=BMM_Additive
    translationSpace=BCS_BoneSpace
    alphaBoolBlend {blendOption=Linear}
    boneToModify {boneName=dyn_leaf_7_2}
    rotation {yaw=-15}
    scale=<0.1 0.1 0.1>
   }
   Output Pose(Result=Component To Local(ComponentPose=@n66)) {layerGroup=None}
  ]}
 EventGraph {nodes [
   Event BlueprintUpdateAnimation()
   > n1 = Sequence()
   > n2 = Branch(Condition=Get bIsInFrontEnd())
   > n3 = Set Dyn_Stem_End_Time(Dyn_Stem_End_Time:double=50) -> (Output_Get:double)
   > Set Dyn_Stem_End_Time_4(Dyn_Stem_End_Time_4:double=float * float(A=@n3, B=0.333) {is=title K2Node_PromotableOperator}) -> (Output_Get:double)
   > Set Dyn_Speed_7(Dyn_Speed_7:double=float * float(A=@n3, B=0.05)) -> (Output_Get:double)
   > Set Dyn_Speed_6(Dyn_Speed_6:double=float * float(A=@n3, B=0.1)) -> (Output_Get:double)
   n4 = Set Dyn_Stem_End_Time(execute=@n2.else, Dyn_Stem_End_Time=900)
   > Set Dyn_Stem_End_Time_4(Dyn_Stem_End_Time_4=float * float(A=@n4, B=0.333))
   > Set Dyn_Speed_7(Dyn_Speed_7=float * float(A=@n4, B=0.1))
   > Set Dyn_Speed_6(Dyn_Speed_6=float * float(A=@n4, B=0.2))
   n5 = Sequence!(execute=@n1.then_1) -> (then_0, then_1, then_2:Exec, then_3:Exec, then_4:Exec, then_5:Exec, then_6:Exec)
   n6 = Sequence!(execute=@n5.then_3) -> (then_0, then_1, then_2, then_3)
   n7 = float * float(A=Get Dyn_Stem_End_Time() -> (Dyn_Stem_End_Time:double), B=0.75)
   n8 = float - float(A=GetMatchTime!() -> (ReturnValue:float), B=Get Dyn_Time_Start() -> (Dyn_Time_Start:double))
   > Set Dyn_Stem_4_1(Dyn_Stem_4_1:double=MapRangeClamped(Value="Clamp (Float)"(Value=@n8, Min=0, Max=@n7), InRangeA=float * float(A=@n7, B=0.42), InRangeB=@n7, OutRangeA=1, OutRangeB=0)) -> (Output_Get:double)
   Set Dyn_Stem_4_2(
    execute=@n6.then_1
    Dyn_Stem_4_2:double=MapRangeClamped(Value="Clamp (Float)"(Value=@n8, Min=0, Max=@n7), InRangeA=0, InRangeB=@n7, OutRangeA=1, OutRangeB=0)
   ) -> (Output_Get:double)
   Set Dyn_Stem_4_3(
    execute=@n6.then_2
    Dyn_Stem_4_3:double=MapRangeClamped(Value="Clamp (Float)"(Value=@n8, Min=0, Max=@n7), InRangeA=0, InRangeB=@n7, OutRangeA=1, OutRangeB=0)
   ) -> (Output_Get:double)
   n9 = Get Dyn_Stem_End_Time_4() -> (Dyn_Stem_End_Time_4:double)
   Set Dyn_Stem_4_4(
    execute=@n6.then_3
    Dyn_Stem_4_4:double=MapRangeClamped(Value="Clamp (Float)"(Value=@n8, Min=0, Max=@n9), InRangeA=0, InRangeB=@n9, OutRangeA=1, OutRangeB=0)
   ) -> (Output_Get:double)
   > n10 = Sequence!(execute) -> (then_0, then_1, then_2, then_3, then_4:Exec)
   n11 = float - float(A=GetMatchTime(), B=Get Dyn_Time_Start())
   n12 = Get Dyn_Stem_End_Time()
   n13 = float - float(A=@n12, B=float * float(A=@n12, B=0.1))
   > Set Dyn_Leaf_4_1(Dyn_Leaf_4_1:double=MapRangeClamped(Value="Clamp (Float)"(Value=@n11, Min=0, Max=@n13), InRangeA=0, InRangeB=@n13, OutRangeA=1, OutRangeB=0)) -> (Output_Get:double)
   n14 = float - float(A=GetMatchTime(), B=Get Dyn_Time_Start())
   n15 = Get Dyn_Stem_End_Time()
   n16 = float - float(A=@n15, B=float * float(A=@n15, B=0.175))
   Set Dyn_Leaf_4_2(
    execute=@n10.then_1
    Dyn_Leaf_4_2:double=MapRangeClamped(Value="Clamp (Float)"(Value=@n14, Min=0, Max=@n16), InRangeA=0, InRangeB=@n16, OutRangeA=1, OutRangeB=0)
   ) -> (Output_Get:double)
   n17 = float - float(A=GetMatchTime(), B=Get Dyn_Time_Start())
   n18 = Get Dyn_Stem_End_Time()
   n19 = float - float(A=@n18, B=float * float(A=@n18, B=0.055))
   Set Dyn_Leaf_4_3(
    execute=@n10.then_2
    Dyn_Leaf_4_3:double=MapRangeClamped(Value="Clamp (Float)"(Value=@n17, Min=0, Max=@n19), InRangeA=0, InRangeB=@n19, OutRangeA=1, OutRangeB=0)
   ) -> (Output_Get:double)
   n20 = float - float(A=GetMatchTime(), B=Get Dyn_Time_Start())
   n21 = Get Dyn_Stem_End_Time()
   n22 = float - float(A=@n21, B=float * float(A=@n21, B=0.85))
   n23 = float * float(A=@n21, B=1)
   Set Dyn_Leaf_4_base(
    execute=@n10.then_3
    Dyn_Leaf_4_base:double=MapRangeClamped(Value="Clamp (Float)"(Value=@n20, Min=0, Max=@n22), InRangeA=float - float(A=@n21, B=@n23), InRangeB=@n22, OutRangeA=1, OutRangeB=0)
   ) -> (Output_Get:double)
   n24 = float - float(A=GetMatchTime(), B=Get Dyn_Time_Start())
   n25 = Get Dyn_Stem_End_Time()
   n26 = float - float(A=@n25, B=float * float(A=@n25, B=0.45))
   n27 = float * float(A=@n25, B=0.95)
   Set Dyn_Leaf_4_d_1(
    execute=@n10.then_4
    Dyn_Leaf_4_d_1:double=MapRangeClamped(Value="Clamp (Float)"(Value=@n24, Min=0, Max=@n26), InRangeA=float - float(A=@n25, B=@n27), InRangeB=@n26, OutRangeA=1, OutRangeB=0)
   ) -> (Output_Get:double)
   n28 = Sequence!(execute=@n5.then_6) -> (then_0, then_1, then_2, then_3)
   n29 = float - float(A=Get Dyn_Stem_End_Time(), B=Get Dyn_Speed_7() -> (Dyn_Speed_7:double))
   n30 = float - float(A=GetMatchTime(), B=Get Dyn_Time_Start())
   > Set Dyn_Stem_7_1(Dyn_Stem_7_1:double=MapRangeClamped(Value="Clamp (Float)"(Value=@n30, Min=0, Max=@n29), InRangeA=float * float(A=@n29, B=0.73), InRangeB=@n29, OutRangeA=1, OutRangeB=0)) -> (Output_Get:double)
   Set Dyn_Stem_7_2(
    execute=@n28.then_1
    Dyn_Stem_7_2:double=MapRangeClamped(Value="Clamp (Float)"(Value=@n30, Min=0, Max=@n29), InRangeA=float * float(A=@n29, B=0.475), InRangeB=@n29, OutRangeA=1, OutRangeB=0)
   ) -> (Output_Get:double)
   Set Dyn_Stem_7_3(
    execute=@n28.then_2
    Dyn_Stem_7_3:double=MapRangeClamped(Value="Clamp (Float)"(Value=@n30, Min=0, Max=@n29), InRangeA=float * float(A=@n29, B=0.25), InRangeB=@n29, OutRangeA=1, OutRangeB=0)
   ) -> (Output_Get:double)
   n31 = float - float(A=Get Dyn_Stem_End_Time_4(), B=Get Dyn_Speed_7())
   Set Dyn_Stem_7_4(
    execute=@n28.then_3
    Dyn_Stem_7_4:double=MapRangeClamped(Value="Clamp (Float)"(Value=@n30, Min=0, Max=@n31), InRangeA=0, InRangeB=@n31, OutRangeA=1, OutRangeB=0)
   ) -> (Output_Get:double)
   > n32 = Sequence!(execute) -> (then_0, then_1)
   n33 = float - float(A=GetMatchTime(), B=Get Dyn_Time_Start())
   n34 = Get Dyn_Stem_End_Time()
   n35 = float - float(A=@n34, B=float * float(A=@n34, B=0.085))
   > Set Dyn_Leaf_7_1(Dyn_Leaf_7_1:double=MapRangeClamped(Value="Clamp (Float)"(Value=@n33, Min=0, Max=@n35), InRangeA=0, InRangeB=@n35, OutRangeA=1, OutRangeB=0)) -> (Output_Get:double)
   n36 = Get Dyn_Stem_End_Time()
   n37 = float - float(A=@n36, B=float * float(A=@n36, B=0.195))
   n38 = float - float(A=GetMatchTime(), B=Get Dyn_Time_Start())
   Set Dyn_Leaf_7_2(
    execute=@n32.then_1
    Dyn_Leaf_7_2:double=MapRangeClamped(Value="Clamp (Float)"(Value=@n38, Min=0, Max=@n37), InRangeA=0, InRangeB=@n37, OutRangeA=1, OutRangeB=0)
   ) -> (Output_Get:double)
   n39 = Sequence!(execute=@n5.then_2) -> (then_0, then_1, then_2:Exec, then_3:Exec)
   n40 = float - float(A=GetMatchTime(), B=Get Dyn_Time_Start())
   n41 = Get Dyn_Stem_End_Time()
   > Set Dyn_Stem_3_1(Dyn_Stem_3_1:double=MapRangeClamped(Value="Clamp (Float)"(Value=@n40, Min=0, Max=@n41), InRangeA=float * float(A=@n41, B=0.67), InRangeB=@n41, OutRangeA=1, OutRangeB=0)) -> (Output_Get:double)
   Set Dyn_Stem_3_2(
    execute=@n39.then_1
    Dyn_Stem_3_2:double=MapRangeClamped(Value="Clamp (Float)"(Value=@n40, Min=0, Max=@n41), InRangeA=float * float(A=@n41, B=0.535), InRangeB=@n41, OutRangeA=1, OutRangeB=0)
   ) -> (Output_Get:double)
   Set Dyn_Stem_3_3(
    execute=@n39.then_2
    Dyn_Stem_3_3:double=MapRangeClamped(Value="Clamp (Float)"(Value=@n40, Min=0, Max=@n41), InRangeA=float * float(A=@n41, B=0.333), InRangeB=@n41, OutRangeA=1, OutRangeB=0)
   ) -> (Output_Get:double)
   n42 = Get Dyn_Stem_End_Time_4()
   Set Dyn_Stem_3_4(
    execute=@n39.then_3
    Dyn_Stem_3_4:double=MapRangeClamped(Value="Clamp (Float)"(Value=@n40, Min=0, Max=@n42), InRangeA=0, InRangeB=@n42, OutRangeA=1, OutRangeB=0)
   ) -> (Output_Get:double)
   n43 = float - float(A=GetMatchTime(), B=Get Dyn_Time_Start())
   n44 = Get Dyn_Stem_End_Time()
   n45 = float - float(A=@n44, B=float * float(A=@n44, B=0.175))
   > Set Dyn_Leaf_3_1(Dyn_Leaf_3_1:double=MapRangeClamped(Value="Clamp (Float)"(Value=@n43, Min=0, Max=@n45), InRangeA=0, InRangeB=@n45, OutRangeA=1, OutRangeB=0)) -> (Output_Get:double)
   n46 = Sequence(execute=@n5.then_1)
   n47 = float - float(A=GetMatchTime(), B=Get Dyn_Time_Start())
   n48 = float * float(A=Get Dyn_Stem_End_Time(), B=0.66)
   > Set Dyn_Stem_2_1(Dyn_Stem_2_1:double=MapRangeClamped(Value="Clamp (Float)"(Value=@n47, Min=0, Max=@n48), InRangeA=float * float(A=@n48, B=0.5), InRangeB=@n48, OutRangeA=1, OutRangeB=0)) -> (Output_Get:double)
   Set Dyn_Stem_2_2(
    execute=@n46.then_1
    Dyn_Stem_2_2:double=MapRangeClamped(Value="Clamp (Float)"(Value=@n47, Min=0, Max=@n48), InRangeA=0, InRangeB=@n48, OutRangeA=1, OutRangeB=0)
   ) -> (Output_Get:double)
   Set Dyn_Stem_2_3(
    execute=@n46.then_2
    Dyn_Stem_2_3:double=MapRangeClamped(Value="Clamp (Float)"(Value=@n47, Min=0, Max=@n48), InRangeA=0, InRangeB=@n48, OutRangeA=1, OutRangeB=0)
   ) -> (Output_Get:double)
   n49 = Get Dyn_Stem_End_Time_4()
   Set Dyn_Stem_2_4(
    execute=@n46.then_3
    Dyn_Stem_2_4:double=MapRangeClamped(Value="Clamp (Float)"(Value=@n47, Min=0, Max=@n49), InRangeA=0, InRangeB=@n49, OutRangeA=1, OutRangeB=0)
   ) -> (Output_Get:double)
   > n50 = Sequence!(execute) -> (then_0, then_1, then_2, then_3, then_4:Exec, then_5:Exec)
   n51 = float - float(A=GetMatchTime(), B=Get Dyn_Time_Start())
   n52 = Get Dyn_Stem_End_Time()
   n53 = float - float(A=@n52, B=float * float(A=@n52, B=0.11))
   > Set Dyn_Leaf_2_1(Dyn_Leaf_2_1:double=MapRangeClamped(Value="Clamp (Float)"(Value=@n51, Min=0, Max=@n53), InRangeA=0, InRangeB=@n53, OutRangeA=1, OutRangeB=0)) -> (Output_Get:double)
   n54 = float - float(A=GetMatchTime(), B=Get Dyn_Time_Start())
   n55 = Get Dyn_Stem_End_Time()
   n56 = float - float(A=@n55, B=float * float(A=@n55, B=0.15))
   Set Dyn_Leaf_2_2(
    execute=@n50.then_1
    Dyn_Leaf_2_2:double=MapRangeClamped(Value="Clamp (Float)"(Value=@n54, Min=0, Max=@n56), InRangeA=0, InRangeB=@n56, OutRangeA=1, OutRangeB=0)
   ) -> (Output_Get:double)
   n57 = float - float(A=GetMatchTime(), B=Get Dyn_Time_Start())
   n58 = Get Dyn_Stem_End_Time()
   n59 = float - float(A=@n58, B=float * float(A=@n58, B=0.09))
   Set Dyn_Leaf_2_3(
    execute=@n50.then_2
    Dyn_Leaf_2_3:double=MapRangeClamped(Value="Clamp (Float)"(Value=@n57, Min=0, Max=@n59), InRangeA=0, InRangeB=@n59, OutRangeA=1, OutRangeB=0)
   ) -> (Output_Get:double)
   n60 = float - float(A=GetMatchTime(), B=Get Dyn_Time_Start())
   n61 = Get Dyn_Stem_End_Time()
   n62 = float - float(A=@n61, B=float * float(A=@n61, B=0.6))
   n63 = float * float(A=@n61, B=0.9)
   Set Dyn_Leaf_2_base(
    execute=@n50.then_3
    Dyn_Leaf_2_base:double=MapRangeClamped(Value="Clamp (Float)"(Value=@n60, Min=0, Max=@n62), InRangeA=float - float(A=@n61, B=@n63), InRangeB=@n62, OutRangeA=1, OutRangeB=0)
   ) -> (Output_Get:double)
   n64 = float - float(A=GetMatchTime(), B=Get Dyn_Time_Start())
   n65 = Get Dyn_Stem_End_Time()
   n66 = float - float(A=@n65, B=float * float(A=@n65, B=0.1))
   n67 = float * float(A=@n65, B=0.65)
   Set Dyn_Leaf_2_d_1(
    execute=@n50.then_4
    Dyn_Leaf_2_d_1:double=MapRangeClamped(Value="Clamp (Float)"(Value=@n64, Min=0, Max=@n66), InRangeA=float - float(A=@n65, B=@n67), InRangeB=@n66, OutRangeA=1, OutRangeB=0)
   ) -> (Output_Get:double)
   n68 = float - float(A=GetMatchTime(), B=Get Dyn_Time_Start())
   n69 = Get Dyn_Stem_End_Time()
   n70 = float - float(A=@n69, B=float * float(A=@n69, B=0.15))
   n71 = float * float(A=@n69, B=0.6)
   Set Dyn_Leaf_2_d_2(
    execute=@n50.then_5
    Dyn_Leaf_2_d_2:double=MapRangeClamped(Value="Clamp (Float)"(Value=@n68, Min=0, Max=@n70), InRangeA=float - float(A=@n69, B=@n71), InRangeB=@n70, OutRangeA=1, OutRangeB=0)
   ) -> (Output_Get:double)
   n72 = Sequence!(execute=@n5.then_0) -> (then_0, then_1)
   n73 = Sequence!(execute=@n72.then_1) -> (then_0, then_1, then_2:Exec, then_3:Exec)
   n74 = float - float(A=GetMatchTime(), B=Get Dyn_Time_Start())
   n75 = Get Dyn_Stem_End_Time()
   > Set Dyn_Stem_1_1(Dyn_Stem_1_1:double=MapRangeClamped(Value="Clamp (Float)"(Value=@n74, Min=0, Max=@n75), InRangeA=float * float(A=@n75, B=0.7), InRangeB=@n75, OutRangeA=1, OutRangeB=0)) -> (Output_Get:double)
   Set Dyn_Stem_1_2(
    execute=@n73.then_1
    Dyn_Stem_1_2:double=MapRangeClamped(Value="Clamp (Float)"(Value=@n74, Min=0, Max=@n75), InRangeA=float * float(A=@n75, B=0.44), InRangeB=@n75, OutRangeA=1, OutRangeB=0)
   ) -> (Output_Get:double)
   Set Dyn_Stem_1_3(
    execute=@n73.then_2
    Dyn_Stem_1_3:double=MapRangeClamped(Value="Clamp (Float)"(Value=@n74, Min=0, Max=@n75), InRangeA=float * float(A=@n75, B=0.33), InRangeB=@n75, OutRangeA=1, OutRangeB=0)
   ) -> (Output_Get:double)
   n76 = Get Dyn_Stem_End_Time_4()
   Set Dyn_Stem_1_4(
    execute=@n73.then_3
    Dyn_Stem_1_4:double=MapRangeClamped(Value="Clamp (Float)"(Value=@n74, Min=0, Max=@n76), InRangeA=0, InRangeB=@n76, OutRangeA=1, OutRangeB=0)
   ) -> (Output_Get:double)
   > n77 = Sequence!(execute) -> (then_0, then_1)
   n78 = float - float(A=GetMatchTime(), B=Get Dyn_Time_Start())
   n79 = Get Dyn_Stem_End_Time()
   n80 = float - float(A=@n79, B=float * float(A=@n79, B=0.35))
   > Set Dyn_Leaf_1_1(Dyn_Leaf_1_1:double=MapRangeClamped(Value="Clamp (Float)"(Value=@n78, Min=0, Max=@n80), InRangeA=0, InRangeB=@n80, OutRangeA=1, OutRangeB=0)) -> (Output_Get:double)
   n81 = Get Dyn_Stem_End_Time()
   n82 = float - float(A=@n81, B=float * float(A=@n81, B=0.225))
   Set Dyn_Leaf_1_2(
    execute=@n77.then_1
    Dyn_Leaf_1_2:double=MapRangeClamped(Value="Clamp (Float)"(Value=@n78, Min=0, Max=@n82), InRangeA=0, InRangeB=@n82, OutRangeA=1, OutRangeB=0)
   ) -> (Output_Get:double)
   n83 = Sequence!(execute=@n5.then_5) -> (then_0, then_1, then_2:Exec, then_3:Exec)
   n84 = float - float(A=Get Dyn_Stem_End_Time(), B=Get Dyn_Speed_6() -> (Dyn_Speed_6:double))
   n85 = float - float(A=GetMatchTime(), B=Get Dyn_Time_Start())
   > Set Dyn_Stem_6_1(Dyn_Stem_6_1:double=MapRangeClamped(Value="Clamp (Float)"(Value=@n85, Min=0, Max=@n84), InRangeA=float * float(A=@n84, B=0.72), InRangeB=@n84, OutRangeA=1, OutRangeB=0)) -> (Output_Get:double)
   Set Dyn_Stem_6_2(
    execute=@n83.then_1
    Dyn_Stem_6_2:double=MapRangeClamped(Value="Clamp (Float)"(Value=@n85, Min=0, Max=@n84), InRangeA=float * float(A=@n84, B=0.49), InRangeB=@n84, OutRangeA=1, OutRangeB=0)
   ) -> (Output_Get:double)
   Set Dyn_Stem_6_3(
    execute=@n83.then_2
    Dyn_Stem_6_3:double=MapRangeClamped(Value="Clamp (Float)"(Value=@n85, Min=0, Max=@n84), InRangeA=float * float(A=@n84, B=0.26), InRangeB=@n84, OutRangeA=1, OutRangeB=0)
   ) -> (Output_Get:double)
   n86 = float - float(A=Get Dyn_Stem_End_Time_4(), B=Get Dyn_Speed_6())
   Set Dyn_Stem_6_4(
    execute=@n83.then_3
    Dyn_Stem_6_4:double=MapRangeClamped(Value="Clamp (Float)"(Value=@n85, Min=0, Max=@n86), InRangeA=0, InRangeB=@n86, OutRangeA=1, OutRangeB=0)
   ) -> (Output_Get:double)
   > n87 = Sequence!(execute) -> (then_0, then_1)
   n88 = float - float(A=GetMatchTime(), B=Get Dyn_Time_Start())
   n89 = Get Dyn_Stem_End_Time()
   n90 = float - float(A=@n89, B=float * float(A=@n89, B=0.11))
   > Set Dyn_Leaf_6_1(Dyn_Leaf_6_1:double=MapRangeClamped(Value="Clamp (Float)"(Value=@n88, Min=0, Max=@n90), InRangeA=0, InRangeB=@n90, OutRangeA=1, OutRangeB=0)) -> (Output_Get:double)
   n91 = float - float(A=GetMatchTime(), B=Get Dyn_Time_Start())
   n92 = Get Dyn_Stem_End_Time()
   n93 = float - float(A=@n92, B=float * float(A=@n92, B=0.225))
   Set Dyn_Leaf_6_2(
    execute=@n87.then_1
    Dyn_Leaf_6_2:double=MapRangeClamped(Value="Clamp (Float)"(Value=@n91, Min=0, Max=@n93), InRangeA=0, InRangeB=@n93, OutRangeA=1, OutRangeB=0)
   ) -> (Output_Get:double)
   n94 = Sequence!(execute=@n5.then_4) -> (then_0, then_1, then_2:Exec, then_3:Exec)
   n95 = Get Dyn_Stem_End_Time()
   n96 = float - float(A=GetMatchTime(), B=Get Dyn_Time_Start())
   > Set Dyn_Stem_5_1(Dyn_Stem_5_1:double=MapRangeClamped(Value="Clamp (Float)"(Value=@n96, Min=0, Max=@n95), InRangeA=float * float(A=@n95, B=0.75), InRangeB=@n95, OutRangeA=1, OutRangeB=0)) -> (Output_Get:double)
   Set Dyn_Stem_5_2(
    execute=@n94.then_1
    Dyn_Stem_5_2:double=MapRangeClamped(Value="Clamp (Float)"(Value=@n96, Min=0, Max=@n95), InRangeA=float * float(A=@n95, B=0.55), InRangeB=@n95, OutRangeA=1, OutRangeB=0)
   ) -> (Output_Get:double)
   Set Dyn_Stem_5_3(
    execute=@n94.then_2
    Dyn_Stem_5_3:double=MapRangeClamped(Value="Clamp (Float)"(Value=@n96, Min=0, Max=@n95), InRangeA=float * float(A=@n95, B=0.333), InRangeB=@n95, OutRangeA=1, OutRangeB=0)
   ) -> (Output_Get:double)
   n97 = Get Dyn_Stem_End_Time_4()
   Set Dyn_Stem_5_4(
    execute=@n94.then_3
    Dyn_Stem_5_4:double=MapRangeClamped(Value="Clamp (Float)"(Value=@n96, Min=0, Max=@n97), InRangeA=0, InRangeB=@n97, OutRangeA=1, OutRangeB=0)
   ) -> (Output_Get:double)
   > n98 = Sequence()
   n99 = float - float(A=GetMatchTime(), B=Get Dyn_Time_Start())
   n100 = Get Dyn_Stem_End_Time()
   n101 = float - float(A=@n100, B=float * float(A=@n100, B=0.805))
   > Set Dyn_Leaf_5_base(Dyn_Leaf_5_base:double=MapRangeClamped(Value="Clamp (Float)"(Value=@n99, Min=0, Max=@n101), InRangeA=0, InRangeB=@n101, OutRangeA=1, OutRangeB=0)) -> (Output_Get:double)
   n102 = float - float(A=GetMatchTime(), B=Get Dyn_Time_Start())
   n103 = Get Dyn_Stem_End_Time()
   n104 = float - float(A=@n103)
   Set Dyn_Leaf_5_1(
    execute=@n98.then_1
    Dyn_Leaf_5_1:double=MapRangeClamped(Value="Clamp (Float)"(Value=@n102, Min=0, Max=@n104), InRangeA=float * float(A=@n103, B=0.15), InRangeB=@n104, OutRangeA=1, OutRangeB=0)
   ) -> (Output_Get:double)
   n105 = float - float(A=GetMatchTime(), B=Get Dyn_Time_Start())
   n106 = Get Dyn_Stem_End_Time()
   n107 = float - float(A=@n106, B=float * float(A=@n106, B=0.11))
   Set Dyn_Leaf_5_2(
    execute=@n98.then_2
    Dyn_Leaf_5_2:double=MapRangeClamped(Value="Clamp (Float)"(Value=@n105, Min=0, Max=@n107), InRangeA=float * float(A=@n106, B=0.075), InRangeB=@n107, OutRangeA=1, OutRangeB=0)
   ) -> (Output_Get:double)
   n108 = float - float(A=GetMatchTime(), B=Get Dyn_Time_Start())
   n109 = Get Dyn_Stem_End_Time()
   n110 = float - float(A=@n109, B=float * float(A=@n109, B=0.055))
   Set Dyn_Leaf_5_3(
    execute=@n98.then_3
    Dyn_Leaf_5_3:double=MapRangeClamped(Value="Clamp (Float)"(Value=@n108, Min=0, Max=@n110), InRangeA=0, InRangeB=@n110, OutRangeA=1, OutRangeB=0)
   ) -> (Output_Get:double)
   Event OnCharacterCustomizationLoaded()
   > Set Dyn_Time_Start(Dyn_Time_Start:double=GetMatchTime()) -> (Output_Get:double)
   CheckForBackbling()
   GetMatchTime()
   TryGetPawnOwner()
   float * float(B=0.42)
   float * float(B=0.75)
   float * float(B=0.6)
   Get Dyn_Time_End() -> (Dyn_Time_End:double)
  ]}
]
variables [
 Dyn_Leaf_1_1 {type=number, default=50}
 Dyn_Leaf_1_2 {type=number, default=50}
 Dyn_Leaf_2_1 {type=number, default=50}
 Dyn_Leaf_2_2 {type=number, default=50}
 Dyn_Leaf_2_3 {type=number, default=50}
 Dyn_Leaf_2_base {type=number, default=1}
 Dyn_Leaf_2_d_1 {type=number, default=1}
 Dyn_Leaf_2_d_2 {type=number, default=1}
 Dyn_Leaf_3_1 {type=number, default=50}
 Dyn_Leaf_4_1 {type=number, default=50}
 Dyn_Leaf_4_2 {type=number, default=50}
 Dyn_Leaf_4_3 {type=number, default=50}
 Dyn_Leaf_4_base {type=number, default=1}
 Dyn_Leaf_4_d_1 {type=number, default=1}
 Dyn_Leaf_5_1 {type=number, default=50}
 Dyn_Leaf_5_2 {type=number, default=50}
 Dyn_Leaf_5_3 {type=number, default=50}
 Dyn_Leaf_5_base {type=number, default=50}
 Dyn_Leaf_6_1 {type=number, default=50}
 Dyn_Leaf_6_2 {type=number, default=50}
 Dyn_Leaf_7_1 {type=number, default=50}
 Dyn_Leaf_7_2 {type=number, default=50}
 Dyn_Speed_6 {type=number, default=50}
 Dyn_Speed_7 {type=number, default=50}
 Dyn_Stem_1_1 {type=number}
 Dyn_Stem_1_2 {type=number}
 Dyn_Stem_1_3 {type=number}
 Dyn_Stem_1_4 {type=number}
 Dyn_Stem_2_1 {type=number}
 Dyn_Stem_2_2 {type=number}
 Dyn_Stem_2_3 {type=number}
 Dyn_Stem_2_4 {type=number}
 Dyn_Stem_3_1 {type=number}
 Dyn_Stem_3_2 {type=number}
 Dyn_Stem_3_3 {type=number}
 Dyn_Stem_3_4 {type=number}
 Dyn_Stem_4_1 {type=number}
 Dyn_Stem_4_2 {type=number}
 Dyn_Stem_4_3 {type=number}
 Dyn_Stem_4_4 {type=number}
 Dyn_Stem_5_1 {type=number}
 Dyn_Stem_5_2 {type=number}
 Dyn_Stem_5_3 {type=number}
 Dyn_Stem_5_4 {type=number}
 Dyn_Stem_6_1 {type=number}
 Dyn_Stem_6_2 {type=number}
 Dyn_Stem_6_3 {type=number}
 Dyn_Stem_6_4 {type=number}
 Dyn_Stem_7_1 {type=number}
 Dyn_Stem_7_2 {type=number}
 Dyn_Stem_7_3 {type=number}
 Dyn_Stem_7_4 {type=number}
 Dyn_Stem_End_Time {type=number, default=300}
 Dyn_Stem_End_Time_4 {type=number, default=100}
 Dyn_Time_End {type=number, default=1}
 Dyn_Time_Start {type=number, default=1}
 Dyn_Value {type=number, default=1}
]
